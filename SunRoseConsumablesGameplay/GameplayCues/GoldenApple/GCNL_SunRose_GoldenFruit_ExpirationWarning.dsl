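; Looping gameplay cue notify that warns the locally viewed, living player that the SunRose Golden Fruit is about to expire.
; On loop start it reads durations and intervals from scalable floats, then begins a self-rescheduling timer pulse.
; Each pulse plays a 2D tick and spawns pelvis-attached FX, switching from long to short intervals as time runs out.
; On removal it clears the timer and fades out warning audio for the local viewer.
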(event EventOnLoopingStart(EitherVFX-preferthisversion)
  (params (MyTarget :type Actor%20Object%20Reference) (Parameters :type Gameplay%20Cue%20Parameters%20Structure%20%28by%20ref%29) (ParticleComponents :type Array%20of%20FXSystem%20Component%20Object%20References) (AudioComponents :type Array%20of%20Audio%20Component%20Object%20References))
  ;[1 1 1] On Looping Start
  (|CallParentFunction MyTarget Parameters ParticleComponents AudioComponents)
  (bind _asfort_player_pawn (Utilities|Casting|CastToFortPlayerPawn MyTarget))
  (Variables|Default|SetPlayer%20Pawn _asfort_player_pawn)
  (if (Default|IsLocalValidPawn (Variables|Default|GetPlayer%20Pawn))
    (Variables|Default|SetTime%20Remaining (ScalableFloat|GetValueatLevel (Variables|Default|GetRow%20Expiration%20Warning%20Duration)))
    (Variables|Default|SetWarning%20Short%20Duration (ScalableFloat|GetValueatLevel (Variables|Default|GetRow%20Expiration%20Short%20Warning%20Duration)))
    (Variables|Default|SetTime%20Interval%20Short (ScalableFloat|GetValueatLevel (Variables|Default|GetRow%20Expiration%20Short%20Warning%20Pulse%20Interval)))
    (Variables|Default|SetTime%20Interval%20Long (ScalableFloat|GetValueatLevel (Variables|Default|GetRow%20Expiration%20Long%20Warning%20Pulse%20Interval)))
    (CallFunction|PulseExpirationWarning)))

(event Custom|PulseExpirationWarning
  ;[1 1 1] Pulse Removal Warning
  (Utilities|FlowControl|Sequence
    (:then_0
      (bind _returnvalue (Audio|SpawnSound2D (Variables|Default|GetPlayer%20Pawn) (Variables|Default|GetSound%20Expiration%20Warning%20Tick)))
      (Variables|Default|SetAudio%20Component%20Expiration%20Warning%20Pulse _returnvalue)
      (if (> (- (Variables|Default|GetTime%20Remaining) (Variables|Default|GetTime%20Interval%20Long)) (Variables|Default|GetWarning%20Short%20Duration))
        (Variables|Default|SetTime%20Interval%20Current (Variables|Default|GetTime%20Interval%20Long))
        (else
          (Variables|Default|SetTime%20Interval%20Current (Variables|Default|GetTime%20Interval%20Short))))
      (Variables|Default|SetTime%20Remaining (- (Variables|Default|GetTime%20Remaining) (Variables|Default|GetTime%20Interval%20Current)))
      (bind _returnvalue_2 (Utilities|Time|SetTimerbyEvent (EventDispatchers|CreateEvent :!event Custom|PulseExpirationWarning) (Variables|Default|GetTime%20Interval%20Current)))
      (Variables|Default|SetHandle%20Timer%20Pulse%20Expiration%20Warning _returnvalue_2))
    (:then_1
      (Effects|SpawnFXSystemAttached (Variables|Default|GetFXSystem%20Asset%20Warning%20NS) (Variables|/Script/Engine.Character|GetMesh (Variables|Default|GetPlayer%20Pawn)) "pelvis"))))

(event EventOnRemoval(EitherVFX-preferthisversion)
  (params (MyTarget :type Actor%20Object%20Reference) (Parameters :type Gameplay%20Cue%20Parameters%20Structure%20%28by%20ref%29) (ParticleComponents :type Array%20of%20FXSystem%20Component%20Object%20References) (AudioComponents :type Array%20of%20Audio%20Component%20Object%20References) (BurstCameraShakeInstance :type Camera%20Shake%20Base%20Object%20Reference) (BurstDecalInstance :type Decal%20Actor%20Object%20Reference))
  ;[1 1 1] On Removal
  (|CallParentFunction MyTarget Parameters ParticleComponents AudioComponents BurstCameraShakeInstance BurstDecalInstance)
  (Utilities|IsValid (Variables|Default|GetPlayer%20Pawn)
    (:Is%20Valid
      _)
    (:Is%20Not%20Valid
      (bind _asfort_player_pawn (Utilities|Casting|CastToFortPlayerPawn MyTarget))
      (Variables|Default|SetPlayer%20Pawn _asfort_player_pawn)))
  (if (Pawn|IsLocallyViewed (Variables|Default|GetPlayer%20Pawn))
    (Utilities|Time|ClearandInvalidateTimerbyHandle (Variables|Default|GetPlayer%20Pawn) (Variables|Default|GetHandle%20Timer%20Pulse%20Expiration%20Warning))
    (Audio|Components|Audio|FadeOut (Variables|Default|GetAudio%20Component%20Expiration%20Warning%20Pulse) 1.5 0.0 "Linear" :!class /Script/Engine.AudioComponent)
    (Audio|Components|Audio|FadeOut (Variables|Default|GetAudio%20Component%20Expiration%20Warning%20Start) 0.25 0.0 "Linear" :!class /Script/Engine.AudioComponent)))

(fn IsLocalValidPawn
  (params (PlayerPawn :type Fort%20Player%20Pawn%20Object%20Reference))
  (return (Math|Boolean|ANDBoolean (Utilities|IsValid PlayerPawn :!class /Script/Engine.KismetSystemLibrary) (not (FortPawn|IsDead PlayerPawn)) (Pawn|IsLocallyViewed PlayerPawn))))

(fn ConstructionScript)
