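; Looping gameplay cue for the SunRose Golden Fruit heal-over-time effect.
; When looping starts, it runs the parent logic, then tones the first Niagara component with fixed overrides of its ScaleBrightness and ScaleAmount float variables.
; BeginPlay, overlap, tick and the construction script are empty.
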(event EventBeginPlay)

(event Collision|EventActorBeginOverlap
  (params (OtherActor :type Actor%20Object%20Reference)))

(event EventTick
  (params (DeltaSeconds :type Float%20%28single-precision%29)))

(event EventOnLoopingStart(Niagara)
  (params (MyTarget :type Actor%20Object%20Reference) (Parameters :type Gameplay%20Cue%20Parameters%20Structure%20%28by%20ref%29) (ParticleComponents :type Array%20of%20Niagara%20Particle%20System%20Component%20Object%20References) (AudioComponents :type Array%20of%20Audio%20Component%20Object%20References))
  (bind _output (Utilities|Array|Get(acopy) ParticleComponents))
  (|CallParentFunction MyTarget Parameters ParticleComponents AudioComponents)
  (Niagara|SetNiagaraVariable(Float) _output "ScaleBrightness" 0.15)
  (Niagara|SetNiagaraVariable(Float) _output "ScaleAmount" 0.2))

(fn ConstructionScript)
